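# Surface material for the castle in the Cattus enemy demo, using Default Lit shading and enabled for skeletal meshes and mesh deformers.
# Its graph holds an albedo texture sample, a normal map sample and two scalar constants, with no procedural logic or exposed parameters.
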Material /Figment_S09/Characters/Enemies/Cattus/Demo/Materials/Demo_Cattus_Castle
graph {
 properties {bUsedWithMeshDeformer=true, bUsedWithSkeletalMesh=true, pixelDepthOffsetMode=PDOM_Legacy, usedShadingModels=Default Lit}
 nodes [
  Constant() {r=0.5}
  Constant() {r=0.6}
  TextureSample() {samplerType=SAMPLERTYPE_Normal, texture=/Figment_S09/Characters/Enemies/Cattus/Demo/Textures/Cattus_Castle_normal}
  TextureSample() {texture=/Figment_S09/Characters/Enemies/Cattus/Demo/Textures/Cattus_Castle_albedo}
 ]
}
